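# Material for the Utopia large speaker screen prop, layering an emissive, audio-reactive equalizer and animated pixel pattern over Fortnite's GigaMix base material.
# The equalizer quantizes UVs into masked cells, drives bars from audio history plus sine-based idle motion, and selects four band colors through Custom HLSL.
# Camera distance adjusts intensity and cell width, and inverse eye adaptation compensates emissive exposure.
# Shared damage, WPO, RVT and low-quality mobile stages follow.
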Material /Hermes/Props/Utopia/Utopia_LargeSpeaker_A/Materials/M_Utopia_LargeSpeakerScreen
graph {
 properties {
  bAutomaticallySetUsageInEditor=false
  bTangentSpaceNormal=false
  bUseMaterialAttributes=true
  bUsedWithInstancedStaticMeshes=true
  bUsedWithNanite=true
  bUsedWithSplineMeshes=true
  ditheredLODTransition=true
  pixelDepthOffsetMode=PDOM_Legacy
  shadingModel=MSM_FromMaterialExpression
  usedShadingModels=Default Lit | Subsurface | Clear Coat | Two Sided Foliage
 }
 nodes [
  n1 = GetMaterialAttributes(None=MaterialFunctionCall() {materialFunction=/Game/Global/Base4Layers_2022/MaterialFunctions/MF_GigaMix_2023}.Result) {attributeGetTypes [#1]}
  n2 = ScalarParameter() {defaultValue=1.5, parameterName=EQ Intensity Near}
  n3 = ScalarParameter() {defaultValue=0.5, parameterName=EQ Intensity Far}
  n4 = NamedRerouteUsage()
  n5 = Add(
   A=Multiply(
    A=MaterialFunctionCall() {materialFunction=/Hermes/Techart/UtopiaEmissiveSurfaces/Club/MF_PixelPattern}.Out
    B=MaterialFunctionCall() {materialFunction=/Hermes/Techart/UtopiaEmissiveSurfaces/Club/MF_ColorFlow}.Out
   )
   B=Multiply(A=NamedRerouteUsage(), B=LinearInterpolate(A=@n2, B=@n3, Alpha=@n4) {constA=4}) {constB=4}
  )
  n6 = SetMaterialAttributes(
   MaterialAttributes=@n1.MaterialAttributes
   Emissive Color=Add(A=@n1.Emissive Color, B=EyeAdaptationInverse(LightValueInput=@n5).EyeAdaptationInverse)
  ) {attributeSetTypes [#1]}
  MaterialExpressionReroute_70 = Reroute(None=@n6)
  n7 = ScalarParameter() {
   desc=Documentation link
   group="0 (Expand)  Documentation (RMB - Copy display name) and Debug view"
   parameterName=https://confluence.it.epicgames.com/display/FORT/Fortnite+Base+Material+2022
  }
  n8 = SetMaterialAttributes!(MaterialAttributes=@MaterialExpressionReroute_70.MaterialAttributes, World Position Offset=Multiply(A=@n7)) {attributeSetTypes [#2]}
  MaterialExpressionReroute_71 = Reroute(None=@n8)
  MaterialExpressionReroute_69 = Reroute(None=@MaterialExpressionReroute_71.Output)
  n9 = MaterialFunctionCall(Material_In=@MaterialExpressionReroute_69.MaterialAttributes) {materialFunction=&MF_FN_DestructibleDamage}
  n10 = MaterialFunctionCall!(
   In=MaterialFunctionCall!(Material Attributes=MaterialFunctionCall!(In=@n9.Material_Out) {materialFunction=/Game/Athena/WinterArt/Materials/MF_WPO_WindNew}.Result) {materialFunction=/Game/Creative/Sets/MilitaryBase/Props/MF_FanSpin}.Result
   UseSphericalMappingOverride
  ) {materialFunction=/Game/Packages/Fortress_SharedMaterials/Base_Material_Functions/MF_CurieFallbackEffect}
  n11 = MaterialFunctionCall!(In=MaterialFunctionCall!(In=MaterialFunctionCall!(In=@n10.Result, Custom Mask, Custom Rotation Axis) {materialFunction=/Game/Athena/WinterArt/Materials/MF_WindWPO_v2}.Result) {materialFunction=/Game/Global/Base4Layers_2022/MaterialFunctions/MF_SnowAndIceArchitecture}.Result) {materialFunction=/Game/Global/Base4Layers_2022/MaterialFunctions/MF_VibrationRattle}
  MaterialExpressionReroute_68 = Reroute(None=@n11.Result)
  Composite!(None=@MaterialExpressionReroute_68.Result) {desc=Transform to World in case we will use RVT after this point, subgraphName=Tangent to World}
  Composite!(MaterialAttributes=@n6) {subgraphName=Documentation Link}
  Composite(MaterialAttributes=@MaterialExpressionReroute_71.Output) {desc=Game Shared WPO and Fire Effect, subgraphName=WPO And Shared Functions}
  MaterialExpressionReroute_1 = Reroute(None=@n18.RVT H)
  MaterialExpressionReroute_3 = Reroute(None=@MaterialExpressionReroute_68.Result)
  n12 = GetMaterialAttributes(None=@MaterialExpressionReroute_3.Input) {attributeGetTypes [#3]}
  n13 = Transform(None=@n12.Normal)
  n14 = LinearInterpolate(A=Multiply(A=@n13) {constB=-1}, B=@n13, Alpha=Saturate(None=TwoSidedSign())) {desc="Fixes normals\r\non backfaces"}
  n15 = SetMaterialAttributes!(
   MaterialAttributes=@n12.MaterialAttributes
   Normal=StaticSwitchParameter(True=MaterialProxyReplace(Realtime=@n14, MaterialProxy=@n12.Normal), False=@n14) {parameterName=Force TangentSpaceNormalsFor HLOD ProxyReplace}
  ) {attributeSetTypes [#3]}
  MaterialExpressionReroute_2 = Reroute(None=@n15)
  n16 = MaterialFunctionCall!(In=MaterialFunctionCall!(In=@MaterialExpressionReroute_2.Output) {materialFunction=/Game/ArtTools/Tech_Art/DirectFX/Assets/MaterialFunctions/MF_DirectFX}.Result) {materialFunction=/Game/ArtTools/Tech_Art/Veneer/MaterialFunctions/MF_Veneer}
  MaterialExpressionReroute_8 = Reroute(None=@n16.Result)
  n17 = MaterialFunctionCall!(AO, RVT_Height=@MaterialExpressionReroute_1."RVT_Height (S)", For Buildings=@MaterialExpressionReroute_8."For Moss (MA)", For Moss) {materialFunction=/Game/Environments/Asteria/Rocks/Materials/MaterialFunctions/MF_JungleBiome}
  MaterialExpressionReroute_0 = Reroute(None=@n17.Buildings Jungl)
  n18 = MaterialFunctionCall!(
   In=@MaterialExpressionReroute_0.Result
   "RVT Height Map (Roughness)"=GetMaterialAttributes(None=@n16.Result) {attributeGetTypes [#4]}.Roughness
  ) {materialFunction=/Game/Global/Base4Layers_2022/MaterialFunctions/MF_RVTBlending}
  Composite!(InputPin=@n18.Result) {subgraphName=Mobile Opt}
  Composite!("For Moss (MA)"=@n16.Result, "RVT_Height (S)"=@n18.RVT H) {subgraphName="Seasonal "}
  MaterialExpressionReroute_7 = Reroute(None=@n18.Result)
  n19 = GetMaterialAttributes(None=@MaterialExpressionReroute_7.InputPin)
  n20 = Constant() {r=0.5}
  n21 = Constant() {r=0.8}
  n22 = ShadingPathSwitch(
   Default=@MaterialExpressionReroute_7.InputPin
   Mobile=QualitySwitch(
    Default=@MaterialExpressionReroute_7.InputPin
    Low=SetMaterialAttributes!(MaterialAttributes=@n19.MaterialAttributes, Specular=@n20, Roughness=@n21) {attributeSetTypes [#5, #4]}
    High=@MaterialExpressionReroute_7.InputPin
    Medium=@MaterialExpressionReroute_7.InputPin
    Epic=@MaterialExpressionReroute_7.InputPin
   )
  )
  MaterialExpressionReroute_6 = Reroute(None=@n22)
  MaterialFunctionCall(MA In=@MaterialExpressionReroute_6.Output) {materialFunction=&MF_CreativeMoveTool}
  n23 = TextureCoordinate()
  n24 = ComponentMask(None=VectorParameter() {parameterName=UV Scale, defaultValue=<rgba 1 0.5 0 1>}.RGB) {g=true, r=true}
  n25 = Add(
   A=MaterialFunctionCall(UVs=@n23, Texture Scale=@n24) {materialFunction=&ScaleUVsByCenter}.UVs
   B=ComponentMask(None=VectorParameter() {parameterName=UV Offset, defaultValue=<rgba 0 0.2 0 1>}.RGB) {g=true, r=true}
  )
  n26 = OneMinus(None=Abs(None=ConstantBiasScale(None=@n25) {bias=-0.5, scale=2}))
  Multiply(A=@n26, B=@n24)
  n27 = Multiply(A=ComponentMask(None=NamedRerouteUsage()) {r=true}) {constB=1.4}
  n28 = Multiply(A=@n27) {constB=0.064}
  n29 = Add(A=Multiply(A=Time(), B=ScalarParameter() {defaultValue=1.2, group=Equalizer, parameterName=EQAmbientSpeed}), B=NamedRerouteUsage())
  n30 = Multiply(A=@n29) {constB=0.6666}
  n31 = ConstantBiasScale(None=Sine(None=Add(A=@n28, B=@n30)))
  n32 = Multiply(A=@n27) {constB=0.9}
  n33 = Multiply(A=@n29)
  n34 = ConstantBiasScale(None=Sine(None=Add(A=@n32, B=@n33)))
  n35 = Multiply(A=Power(Base=Add(A=@n31, B=@n34)) {constExponent=0.5, desc=higher more often}) {constB=0.8}
  n36 = Multiply(A=Power(Base=Saturate(None=@n35)), B=ScalarParameter() {defaultValue=0.7, group=Equalizer, parameterName=EQAmbientHeight}) {constB=0.6666, desc=ambient}
  n37 = NamedRerouteUsage()
  n38 = OneMinus(None=ScalarParameter() {defaultValue=0.1, group=Equalizer, parameterName=EQAmbientHeightDuringReactive, sliderMax=1})
  Multiply(A=@n36, B=OneMinus(None=Multiply(A=@n37, B=@n38) {constB=0.9})) {desc=reduce ambient when music starts}
  n39 = VectorParameter() {group=Equalizer, parameterName=EQColor_Bass, sortPriority=70, defaultValue=<rgba 0.05 0 1.5 1>}
  n40 = NamedRerouteUsage()
  MaterialExpressionStaticSwitchParameter_4 = StaticSwitchParameter(
   True=LinearInterpolate(
    A=VectorParameter() {group=Equalizer, parameterName=EQColor_Bass_Blend, sortPriority=71, defaultValue=<rgba 1 1 1 1>}.RGB
    B=@n39.RGB
    Alpha=@n40
   )
   False=@n39.RGB
  ) {group=Equalizer, parameterName=EQUseBlendedColor, sortPriority=0}
  n41 = VectorParameter() {group=Equalizer, parameterName=EQColor_LowerMid_Blend, sortPriority=81, defaultValue=<rgba 1 1 1 1>}
  n42 = VectorParameter() {group=Equalizer, parameterName=EQColor_LowerMid, sortPriority=80, defaultValue=<rgba 1.4 0 0.14 1>}
  MaterialExpressionStaticSwitchParameter_5 = StaticSwitchParameter(True=LinearInterpolate(A=@n41.RGB, B=@n42.RGB, Alpha=@n40), False=@n42.RGB) {group=Equalizer, parameterName=EQUseBlendedColor, sortPriority=0}
  n43 = VectorParameter() {group=Equalizer, parameterName=EQColor_UpperMid_Blend, sortPriority=91, defaultValue=<rgba 1 1 1 1>}
  n44 = VectorParameter() {group=Equalizer, parameterName=EQColor_UpperMid, sortPriority=90, defaultValue=<rgba 0 0.8333 1 1>}
  MaterialExpressionStaticSwitchParameter_6 = StaticSwitchParameter(True=LinearInterpolate(A=@n43.RGB, B=@n44.RGB, Alpha=@n40), False=@n44.RGB) {group=Equalizer, parameterName=EQUseBlendedColor, sortPriority=0}
  n45 = VectorParameter() {group=Equalizer, parameterName=EQColor_Treble_Blend, sortPriority=101, defaultValue=<rgba 1 1 1 1>}
  n46 = VectorParameter() {group=Equalizer, parameterName=EQColor_Treble, sortPriority=100, defaultValue=<rgba 0 0 1 1>}
  MaterialExpressionStaticSwitchParameter_7 = StaticSwitchParameter(True=LinearInterpolate(A=@n45.RGB, B=@n46.RGB, Alpha=@n40), False=@n46.RGB) {group=Equalizer, parameterName=EQUseBlendedColor, sortPriority=0}
  MaterialExpressionNamedRerouteUsage_5 = NamedRerouteUsage()
  n47 = Custom!(
   in1=@MaterialExpressionStaticSwitchParameter_4
   in2=@MaterialExpressionStaticSwitchParameter_5
   in3=@MaterialExpressionStaticSwitchParameter_6
   in4=@MaterialExpressionStaticSwitchParameter_7
   alpha=@MaterialExpressionNamedRerouteUsage_5
  ) {
   code="##pragma once\r\n\r\nconst int numVals = 4;\r\n\r\nconst float3 values[numVals] = {\r\n\t\tin1, \r\n        in2,\r\n\t\tin3, \r\n        in4\r\n\t};\r\n\r\nconst uint a = uint(alpha);\r\n\r\nreturn values[a];"
   inputs [
    {
     inputName=in1
     input {expression=&self:MaterialExpressionStaticSwitchParameter_4}
    }
    {
     inputName=in2
     input {expression=&self:MaterialExpressionStaticSwitchParameter_5}
    }
    {
     inputName=in3
     input {expression=&self:MaterialExpressionStaticSwitchParameter_6}
    }
    {
     inputName=in4
     input {expression=&self:MaterialExpressionStaticSwitchParameter_7}
    }
    {
     inputName=alpha
     input {expression=&self:MaterialExpressionNamedRerouteUsage_5}
    }
   ]
  }
  n48 = NamedRerouteUsage()
  n49 = OneMinus(None=Add(
    A=ChannelMaskParameter(None=@n48) {maskChannel=Green, parameterName=EQLevels, defaultValue=<rgba 0 1 0 0>}.R
    B=ScalarParameter() {defaultValue=0.15, group=Equalizer, parameterName=EQBias}
   ) {constB=0.08, desc="pivot (increase to raise bottom"})
  n50 = Multiply(A=Add(A=Subtract(B=@n49) {constA=0}, B=@n36), B=NamedRerouteUsage())
  n51 = Saturate(None=Multiply(A=NamedRerouteUsage(), B=SmoothStep(Value=@n50) {constMax=0.05, constMin=0.005}))
  Multiply(A=@n47, B=@n51)
  n52 = MaterialFunctionCall(Float2=NamedRerouteUsage()) {materialFunction=&BreakOutFloat2Components}
  n53 = Multiply(A=@n52.R, B=ScalarParameter() {defaultValue=8, group=Equalizer, parameterName=EQColumns, sortPriority=10})
  n54 = ScalarParameter() {defaultValue=8, group=Equalizer, parameterName=EQRows, sortPriority=10}
  n55 = Multiply(A=@n52.G, B=@n54)
  NamedRerouteDeclaration(None=Frac(None=MaterialFunctionCall(X=@n53, Y=@n55) {materialFunction=&MakeFloat2}.Result)) {name=CellUVCoords, nodeColor=<rgba 1 0.5391 0 1>}
  n56 = Multiply(A=NamedRerouteUsage()) {constB=2}
  n57 = MaterialFunctionCall(Float2=Abs(None=Subtract(A=@n56))) {materialFunction=&BreakOutFloat2Components}
  n58 = SphereMask(
   A=@n57.R
   B=Constant()
   Radius=LinearInterpolate(
    A=ScalarParameter() {defaultValue=0.5, group=Equalizer, parameterName=EQCellWidth, sortPriority=15}
    B=ScalarParameter() {defaultValue=0.75, group=Equalizer, parameterName=EQCellWidth Far, sortPriority=15}
    Alpha=NamedRerouteUsage()
   )
   Hardness=ScalarParameter() {parameterName=EQHardness}
  ) {attenuationRadius=0.4}
  n59 = Constant()
  n60 = ScalarParameter() {defaultValue=1, group=Equalizer, parameterName=EQCellHeight, sortPriority=15}
  NamedRerouteDeclaration(None=Multiply(A=SmoothStep(Value=@n58) {constMax=1.5}, B=SphereMask(A=@n57.G, B=@n59, Radius=@n60) {attenuationRadius=0.4})) {name=CellsMask, nodeColor=<rgba 1 0 0.3281 1>}
  n61 = MaterialFunctionCall(X=Floor(None=@n53), Y=@n55) {materialFunction=&MakeFloat2}
  NamedRerouteDeclaration(None=@n61.Result) {name=QuantizedUVs, nodeColor=<rgba 1 0.3047 0 1>}
  NamedRerouteDeclaration(None=Divide(A=@n61.Result, B=@n54)) {name=QuantizedUVsNormalized, nodeColor=<rgba 0 0.5312 1 1>}
  n62 = Subtract(
   A=Length(None=WorldPosition() {worldPositionShaderOffset=WPT_CameraRelativeNoOffsets}.XYZ) {is=class MaterialExpressionLength}
   B=ScalarParameter() {defaultValue=500, parameterName=EQ Fade Distance}
  ) {constB=1500}
  n63 = ScalarParameter() {defaultValue=1500, parameterName=EQ Fade Size}
  NamedRerouteDeclaration(None=Saturate(None=Divide(A=@n62, B=@n63) {constB=2000})) {name=EQ Fade, nodeColor=<rgba 1 0.1641 0 1>}
  n64 = ComponentMask(None=NamedRerouteUsage()) {r=true}
  n65 = ScalarParameter() {
   defaultValue=0.95
   desc="0-1 | The higher this values is, the faster the EQ will  react but it will also have less variation between bars."
   group=Equalizer
   parameterName=EQResponsiveness
   sliderMax=1
  }
  n66 = OneMinus(None=Saturate(None=@n65))
  n67 = MaterialFunctionCall!(useSmoothStep, uvx=Saturate(None=Multiply(A=@n64, B=@n66) {constB=0.05})) {materialFunction=/Game/Accessories/FORT_Backpacks/Backpack_F_MED_Cactus_Rocker/Materials/SampleAudioMPCHistory}
  NamedRerouteDeclaration(None=@n67.Interpolated Amplitude Music) {name=amp, nodeColor=<rgba 0 0.2031 1 1>}
  n68 = NamedRerouteUsage()
  NamedRerouteDeclaration(None=Power(Base=ComponentMask(None=@n68) {g=true}, Exp=ScalarParameter() {defaultValue=1, group=Equalizer, parameterName=EQGradientPower}) {constExponent=0.6667}) {name=ramp, nodeColor=<rgba 0.5579 0 1 1>}
  n69 = ComponentMask(None=@n68) {r=true}
  NamedRerouteDeclaration(None=Floor(None=Multiply(A=@n69) {constB=4})) {name=index, nodeColor=<rgba 0 1 0.3672 1>}
  NamedRerouteDeclaration(None=@n51) {name=EQ Out, nodeColor=<rgba 0 0.6016 1 1>}
  NamedRerouteDeclaration(None=@n67.Interpolated Spectrum) {name=spectrum, nodeColor=<rgba 1 0 0.9609 1>}
  n70 = NamedRerouteUsage()
  n71 = VectorParameter() {group=Equalizer, parameterName="EQPosition(RG)Scale(BA)", sortPriority=0, defaultValue=<rgba 0 0 1 1>}
  n72 = AppendVector(A=@n71.R, B=@n71.R)
  n73 = MaterialFunctionCall(
   UVs=MaterialFunctionCall(
    UVs=Add(A=@n70, B=@n72)
    "Rotation Angle (0-1)"=ScalarParameter() {defaultValue=0.25, group=Equalizer, parameterName=EQRotation, sortPriority=0}
   ) {materialFunction=&CustomRotator}.Rotated Values
   Texture Scale=AppendVector(A=@n71.B, B=@n71.B)
  ) {materialFunction=&ScaleUVsByCenter}
  NamedRerouteDeclaration(None=@n73.0-1 mask) {name=UVMask_0-1, nodeColor=<rgba 1 0 0.2639 1>}
  NamedRerouteDeclaration(None=@n26) {name=BaseUVs, nodeColor=<rgba 1 0 1 1>}
  NamedRerouteDeclaration(None=@n73.UVs) {name=TransformedUVs, nodeColor=<rgba 0.4066 0 1 1>}
  PinBase() {reroutePins [Input {expression=&self:MaterialExpressionReroute_3}]}
  PinBase!(Output=@n15) {reroutePins [Output {expression=&self:MaterialExpressionReroute_2}]}
  PinBase!() {reroutePins [
    "For Moss (MA)" {expression=&self:MaterialExpressionReroute_8}
    "RVT_Height (S)" {expression=&self:MaterialExpressionReroute_1}
   ]}
  PinBase!(Result=@n17.Buildings Jungl) {reroutePins [Result {expression=&self:MaterialExpressionReroute_0}]}
  PinBase!() {reroutePins [MaterialAttributes {expression=&self:MaterialExpressionReroute_70}]}
  PinBase!(Output=@n8) {reroutePins [Output {expression=&self:MaterialExpressionReroute_71}]}
  PinBase!() {reroutePins [MaterialAttributes {expression=&self:MaterialExpressionReroute_69}]}
  PinBase!(Result=@n11.Result) {reroutePins [Result {expression=&self:MaterialExpressionReroute_68}]}
  PinBase!() {reroutePins [InputPin {expression=&self:MaterialExpressionReroute_7}]}
  PinBase!(Output=@n22) {reroutePins [Output {expression=&self:MaterialExpressionReroute_6}]}
  StaticSwitchParameter(True=TextureCoordinate() {coordinateIndex=1}, False=@n23) {group=Equalizer, parameterName=EQUseUVChannel1, sortPriority=0}
  n74 = Saturate(None=NamedRerouteUsage())
  MaterialExpressionComponentMask_4 = ComponentMask(None=@n74) {r=true}
  MaterialExpressionComponentMask_5 = ComponentMask(None=@n74) {g=true}
  MaterialExpressionComponentMask_6 = ComponentMask(None=@n74) {b=true}
  MaterialExpressionComponentMask_7 = ComponentMask(None=@n74) {a=true}
  MaterialExpressionNamedRerouteUsage_6 = NamedRerouteUsage()
  n75 = Custom(
   in1=@MaterialExpressionComponentMask_4
   in2=@MaterialExpressionComponentMask_5
   in3=@MaterialExpressionComponentMask_6
   in4=@MaterialExpressionComponentMask_7
   alpha=@MaterialExpressionNamedRerouteUsage_6
  ) {
   code="##pragma once\r\n\r\nconst int numVals = 4;\r\n\r\n const float values[numVals] = {\r\n\t\tin1, \r\n        in2,\r\n\t\tin3, \r\n        in4\r\n\t};\r\nconst uint a = uint(alpha);\r\n\r\n\r\nreturn values[a];"
   outputType=CMOT_Float1
   inputs [
    {
     inputName=in1
     input {expression=&self:MaterialExpressionComponentMask_4}
    }
    {
     inputName=in2
     input {expression=&self:MaterialExpressionComponentMask_5}
    }
    {
     inputName=in3
     input {expression=&self:MaterialExpressionComponentMask_6}
    }
    {
     inputName=in4
     input {expression=&self:MaterialExpressionComponentMask_7}
    }
    {
     inputName=alpha
     input {expression=&self:MaterialExpressionNamedRerouteUsage_6}
    }
   ]
  }
  n76 = Power(Base=@n75, Exp=ScalarParameter() {defaultValue=1.3333, group=Equalizer, parameterName=EQPower}) {constExponent=1.3333}
  StaticSwitchParameter(
   True=Multiply(A=CollectionParameter() {collection=/Game/Athena/Events/AudioAnalysis/MPC_Audio_Reactive_Parameters, parameterName=Amplitude_Music}, B=@n76)
   False=@n76
  ) {group=Equalizer, parameterName=EQMultiplyByAmplitude}
 ]
}
